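; BP_TextureGraphBugExample is a Blueprint in a ReproFolder, apparently a minimal reproduction of a Texture Graph issue.
; Its construction script sets the Texture Graph's "Texture" parameter, renders it and stores an output in GeneratedFromTextureGraph.
; It then creates a dynamic instance of the ExampleIssue material on the static mesh and passes that texture to a parameter named "None".
; BeginPlay, overlap and Tick events are empty.
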(event EventBeginPlay)

(event Collision|EventActorBeginOverlap
  (params (OtherActor :type Actor%20Object%20Reference)))

(event EventTick
  (params (DeltaSeconds :type Float%20%28single-precision%29)))

(fn ConstructionScript
  (bind _texture_graph (Variables|Default|GetTexture%20Graph))
  (TextureGraph|SetTextureParameterValue _texture_graph "Texture" (Variables|Default|GetParameter%20Value))
  (bind _returnvalue (TextureGraph|RenderTextureGraph _texture_graph))
  (Variables|Default|SetGeneratedFromTextureGraph (Utilities|Array|Get(acopy) _returnvalue))
  (bind _returnvalue_2 (Rendering|Material|CreateDynamicMaterialInstance (Variables|Default|GetStaticMesh) 0 "/Game/NeverCook/ArtTools/Tech_Art/LowPolyTurntable/ReproFolder/ExampleIssue.ExampleIssue" "None" :!class /Script/Engine.PrimitiveComponent))
  (Rendering|Material|SetTextureParameterValue _returnvalue_2 "None" (Variables|Default|GetGeneratedFromTextureGraph)))
